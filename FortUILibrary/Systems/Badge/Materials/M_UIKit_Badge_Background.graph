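# Translucent UI-domain material in the UIKit badge system, named as the background for badges.
# It calls the shared MF_UI_CoreUI_RoundedMask function with constant corner radius inputs to produce a rounded shape mask.
# A Color vector parameter exposes the tint to material instances.
# The graph does not show how the mask and color connect to the material outputs.
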Material /FortUILibrary/Systems/Badge/Materials/M_UIKit_Badge_Background
graph {
 properties {
  bUsedWithStaticMesh=false
  blendMode=BLEND_Translucent
  materialDomain=MD_UI
  pixelDepthOffsetMode=PDOM_Legacy
  usedShadingModels=Default Lit
 }
 nodes [
  MaterialFunctionCall(Corner Radius=Constant() {r=0.5}, Radius in Pixels=Constant()) {materialFunction=&MF_UI_CoreUI_RoundedMask}
  VectorParameter() {parameterName=Color, defaultValue=<rgba 1 1 1 0>}
 ]
}
